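# Cosmetic AnimBP for the KoboLobo player body, adding secondary motion to shoelaces and right-leg jeans pieces.
# Pose copied from the parent mesh runs through a Control Rig, then weighted multi-bone constraints position lace and jeans bones.
# Additive bone rotations adjust laces, mostly blended in while crouched and stationary, before Anim Dynamics simulates laces and angular-sprung jeans.
# The update event sets that crouch flag and computes gravity override via a deprecated function.
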AnimBlueprint /BRCosmetics/Characters/Player/Male/Medium/Bodies/M_MED_KoboLobo/Meshes/M_MED_KoboLobo_AnimBP
target_skeleton=&/M_MED_KoboLobo_Skeleton
graphs [
 AnimGraph {nodes [
   n1 = Control Rig(Source=Copy Pose From Mesh() {bUseAttachedParent=true}) {
    alphaBoolBlend {blendOption=Linear}
    controlRigAssetReference {blueprintRigClass=&/M_MED_KoboLobo_CtrlRig.M_MED_KoboLobo_CtrlRig_C}
   }
   n2 = "Constraint - lace_A_l_MSTR (PP)"!(ComponentPose=Local To Component(LocalPose=@n1), Alpha=1, ConstraintWeights_0=0.916, ConstraintWeights_1:float=0.084) -> (Pose) {
    alphaBoolBlend {blendOption=Linear}
    boneToModify {boneName=lace_A_l_MSTR}
    constraintSetup [{targetBone {boneName=calf_twist_01_l}}, {targetBone {boneName=foot_l}}]
    constraintWeights [1,1]
   }
   n3 = "Constraint - lace_B_l_MSTR (PP)"!(ComponentPose=@n2, Alpha=1, ConstraintWeights_0=0.922, ConstraintWeights_1:float=0.078) -> (Pose) {
    alphaBoolBlend {blendOption=Linear}
    boneToModify {boneName=lace_B_l_MSTR}
    constraintSetup [{targetBone {boneName=calf_twist_01_l}}, {targetBone {boneName=foot_l}}]
    constraintWeights [1,1]
   }
   n4 = "Constraint - lace_A_r_MSTR (PP)"!(ComponentPose=@n3, Alpha=1, ConstraintWeights_0=0.922, ConstraintWeights_1:float=0.078) -> (Pose) {
    alphaBoolBlend {blendOption=Linear}
    boneToModify {boneName=lace_A_r_MSTR}
    constraintSetup [{targetBone {boneName=calf_twist_01_r}}, {targetBone {boneName=foot_r}}]
    constraintWeights [1,1]
   }
   n5 = "Constraint - lace_B_r_MSTR (PP)"!(ComponentPose=@n4, Alpha=1, ConstraintWeights_0=0.916, ConstraintWeights_1:float=0.084) -> (Pose) {
    alphaBoolBlend {blendOption=Linear}
    boneToModify {boneName=lace_B_r_MSTR}
    constraintSetup [{targetBone {boneName=calf_twist_01_r}}, {targetBone {boneName=foot_r}}]
    constraintWeights [1,1]
   }
   n6 = "Constraint - jeans_piece_01_MSTR (PPPPP)"!(
    ComponentPose=@n5
    Alpha=1
    ConstraintWeights_0=0.54
    ConstraintWeights_1:float=0.042
    ConstraintWeights_2:float=0.112
    ConstraintWeights_3:float=0.305
    ConstraintWeights_4:float=0.001
   ) -> (Pose) {
    alphaBoolBlend {blendOption=Linear}
    boneToModify {boneName=jeans_piece_01_MSTR}
    constraintSetup [
     {targetBone {boneName=calf_r}}
     {targetBone {boneName=calf_twist_02_r}}
     {targetBone {boneName=deform_knee_out_r}}
     {targetBone {boneName=thigh_r}}
     {targetBone {boneName=thigh_twist_01_r}}
    ]
    constraintWeights [1,1,1,1,1]
   }
   n7 = "Constraint - jeans_piece_02_MSTR (PPPPP)"!(
    ComponentPose=@n6
    Alpha=1
    ConstraintWeights_0=0.481
    ConstraintWeights_1:float=0.075
    ConstraintWeights_2:float=0.033
    ConstraintWeights_3:float=0.121
    ConstraintWeights_4:float=0.29
   ) -> (Pose) {
    alphaBoolBlend {blendOption=Linear}
    boneToModify {boneName=jeans_piece_02_MSTR}
    constraintSetup [
     {targetBone {boneName=calf_r}}
     {targetBone {boneName=calf_twist_02_r}}
     {targetBone {boneName=deform_knee_out_r}}
     {targetBone {boneName=thigh_r}}
     {targetBone {boneName=jeans_piece_02}}
    ]
    constraintWeights [1,1,1,1,1]
   }
   n8 = "Transform (Modify) Bone - Bone: lace_A_01_r"(ComponentPose=@n7, Alpha=1, Translation="0,0,0", Rotation="0,0,-15", Scale="1,1,1") {rotationMode=BMM_Additive, alphaBoolBlend {blendOption=Linear}, boneToModify {boneName=lace_A_01_r}, rotation {roll=-15}}
   n9 = Get IsCrouchingNotMoving() -> (IsCrouchingNotMoving:bool)
   n10 = "Transform (Modify) Bone - Bone: lace_B_r_MSTR"!(ComponentPose=@n8, bAlphaBoolEnabled:bool=@n9, Translation="0,0,0", Rotation="0,0,-70", Scale="1,1,1") -> (Pose) {
    alphaInputType=Bool
    rotationMode=BMM_Additive
    alphaBoolBlend {blendInTime=0.2, blendOption=CubicInOut, blendOutTime=0.2}
    boneToModify {boneName=lace_B_r_MSTR}
    rotation {roll=-70}
   }
   n11 = "Transform (Modify) Bone - Bone: lace_B_01_r"!(ComponentPose=@n10, bAlphaBoolEnabled:bool=@n9, Translation="0,0,0", Rotation="-30,0,0", Scale="1,1,1") -> (Pose) {
    alphaInputType=Bool
    rotationMode=BMM_Additive
    alphaBoolBlend {blendInTime=0.2, blendOption=CubicInOut, blendOutTime=0.2}
    boneToModify {boneName=lace_B_01_r}
    rotation {pitch=-30}
   }
   n12 = "Transform (Modify) Bone - Bone: lace_A_01_r"!(ComponentPose=@n11, bAlphaBoolEnabled:bool=@n9, Translation="0,0,0", Rotation="-20,16.5,-44", Scale="1,1,1") -> (Pose) {
    alphaInputType=Bool
    rotationMode=BMM_Additive
    alphaBoolBlend {blendInTime=0.2, blendOption=CubicInOut, blendOutTime=0.2}
    boneToModify {boneName=lace_A_01_r}
    rotation=<rot -20 16.5 -44>
   }
   n13 = "Transform (Modify) Bone - Bone: lace_A_01_l"!(ComponentPose=@n12, bAlphaBoolEnabled:bool=@n9, Translation="0,0,0", Rotation="5,-3,-34", Scale="1,1,1") -> (Pose) {
    alphaInputType=Bool
    rotationMode=BMM_Additive
    alphaBoolBlend {blendInTime=0.2, blendOption=CubicInOut, blendOutTime=0.2}
    boneToModify {boneName=lace_A_01_l}
    rotation=<rot 5 -3 -34>
   }
   n14 = "Transform (Modify) Bone - Bone: lace_B_01_l"!(ComponentPose=@n13, bAlphaBoolEnabled:bool=@n9, Translation="0,0,0", Rotation="0,0,-30", Scale="1,1,1") -> (Pose) {
    alphaInputType=Bool
    rotationMode=BMM_Additive
    alphaBoolBlend {blendInTime=0.2, blendOption=CubicInOut, blendOutTime=0.2}
    boneToModify {boneName=lace_B_01_l}
    rotation {roll=-30}
   }
   n15 = Anim Dynamics - Bone: lace_A_02_l!(ComponentPose=@n14, Alpha=1, GravityOverride:Vector="0,0,0", ExternalForce="0,0,0") -> (Pose) {
    angularDampingOverride=0.99
    bOverrideAngularDamping=true
    bUseGravityOverride=true
    alphaBoolBlend {blendOption=Linear}
    boundBone {boneName=lace_A_02_l}
    componentAppliedLinearAccClamp=<200 200 200>
    componentLinearAccScale=<0.4 0.4 0.4>
    componentLinearVelScale=<0.5 0.5 0.5>
    physicsBodyDefinitions [{
      boundBone {boneName=lace_A_02_l}
      boxExtents=<10 5 5>
      constraintSetup {angularLimitsMax=<5 5 80>, angularLimitsMin=<-5 -100 -80>, angularTarget=<0 0 0>}
      localJointOffset=<0.9 0 0>
     }]
   }
   n16 = Anim Dynamics - Bone: lace_B_02_l!(ComponentPose=@n15, Alpha=1, GravityOverride:Vector="0,0,0", ExternalForce="0,0,0") -> (Pose) {
    angularDampingOverride=0.998
    bOverrideAngularDamping=true
    bUseGravityOverride=true
    alphaBoolBlend {blendOption=Linear}
    boundBone {boneName=lace_B_02_l}
    componentAppliedLinearAccClamp=<100 100 100>
    componentLinearAccScale=<0.4 0.4 0.4>
    componentLinearVelScale=<0.5 0.5 0.5>
    physicsBodyDefinitions [{
      boundBone {boneName=lace_B_02_l}
      boxExtents=<8 5 5>
      constraintSetup {angularLimitsMax=<5 5 80>, angularLimitsMin=<-5 -100 -80>, angularTarget=<0 0 0>}
      localJointOffset=<1.1 0 0>
     }]
   }
   n17 = Anim Dynamics - Bone: lace_A_02_r!(ComponentPose=@n16, Alpha=1, GravityOverride:Vector="0,0,0", ExternalForce="0,0,0") -> (Pose) {
    angularDampingOverride=0.99
    bOverrideAngularDamping=true
    bUseGravityOverride=true
    alphaBoolBlend {blendOption=Linear}
    boundBone {boneName=lace_A_02_r}
    componentAppliedLinearAccClamp=<200 200 200>
    componentLinearAccScale=<0.4 0.4 0.4>
    componentLinearVelScale=<0.5 0.5 0.5>
    physicsBodyDefinitions [{
      boundBone {boneName=lace_A_02_r}
      boxExtents=<10 5 5>
      constraintSetup {angularLimitsMax=<5 5 80>, angularLimitsMin=<-5 -100 -80>, angularTarget=<0 0 0>}
      localJointOffset=<0.9 0 0>
     }]
   }
   n18 = Anim Dynamics - Bone: lace_B_02_r!(ComponentPose=@n17, Alpha=1, GravityOverride:Vector="0,0,0", ExternalForce="0,0,0") -> (Pose) {
    angularDampingOverride=0.998
    bOverrideAngularDamping=true
    bUseGravityOverride=true
    alphaBoolBlend {blendOption=Linear}
    boundBone {boneName=lace_B_02_r}
    componentAppliedLinearAccClamp=<100 100 100>
    componentLinearAccScale=<0.4 0.4 0.4>
    componentLinearVelScale=<0.5 0.5 0.5>
    physicsBodyDefinitions [{
      boundBone {boneName=lace_B_02_r}
      boxExtents=<8 5 5>
      constraintSetup {angularLimitsMax=<5 5 80>, angularLimitsMin=<-5 -100 -80>, angularTarget=<0 0 0>}
      localJointOffset=<1.1 0 0>
     }]
   }
   n19 = Anim Dynamics - Bone: jeans_piece_01!(ComponentPose=@n18, Alpha=1, GravityOverride:Vector="0,0,0", ExternalForce="0,0,0") -> (Pose) {
    angularDampingOverride=0.9998
    angularSpringConstant=2000
    bAngularSpring=true
    bOverrideAngularDamping=true
    bUseGravityOverride=true
    alphaBoolBlend {blendOption=Linear}
    boundBone {boneName=jeans_piece_01}
    componentAppliedLinearAccClamp=<500 500 500>
    componentLinearAccScale=<1 1 1>
    componentLinearVelScale=<1 1 1>
    physicsBodyDefinitions [{
      boundBone {boneName=jeans_piece_01}
      boxExtents=<15 10 10>
      constraintSetup {angularLimitsMax=<0 20 60>, angularLimitsMin=<0 -20 -10>, angularTarget=<1 0 0>}
      localJointOffset=<1 0 0>
     }]
   }
   n20 = Anim Dynamics - Bone: jeans_piece_02!(ComponentPose=@n19, Alpha=1, GravityOverride:Vector="0,0,0", ExternalForce="0,0,0") -> (Pose) {
    angularDampingOverride=0.9999
    angularSpringConstant=1500
    bAngularSpring=true
    bOverrideAngularDamping=true
    bUseGravityOverride=true
    alphaBoolBlend {blendOption=Linear}
    boundBone {boneName=jeans_piece_02}
    componentAppliedLinearAccClamp=<100 100 100>
    componentLinearAccScale=<1 1 1>
    componentLinearVelScale=<1 1 1>
    physicsBodyDefinitions [{
      boundBone {boneName=jeans_piece_02}
      boxExtents=<10 10 10>
      constraintSetup {angularLimitsMax=<0 20 60>, angularLimitsMin=<0 -20 -10>, angularTarget=<1 0 0>}
      localJointOffset=<1 0 0>
     }]
   }
   Output Pose(Result=Component To Local(ComponentPose=@n20)) {layerGroup=None}
  ]}
 EventGraph {nodes [
   Event BlueprintUpdateAnimation()
   > n1 = Gravity Override State Function DEPRECATED(CharacterPartAnimInstance=Self-Reference(), GameTimeInSeconds=GetGameTimeInSeconds(), GravityOverrideParams=Get Gravity Override Params())
   > Set Gravity Override(Gravity Override=@n1)
   n2 = Get bIsMoving2D()
   > Set IsCrouchingNotMoving(IsCrouchingNotMoving:bool=AND Boolean(A=Get bIsCrouching(), B=NOT Boolean(A=@n2))) -> (Output_Get:bool)
  ]}
]
variables [
 Gravity Override {type=Vector}
 Gravity Override Params {type=GravityOverrideParamsStruct}
 IsCrouchingNotMoving {type=boolean}
]
